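# Translucent, lit particle material for water splash ripples, usable on sprite and Niagara sprite particles.
# Particle alpha acts as an erosion threshold on the ripple texture's blue channel, multiplied by red and one-minus-alpha, so rising alpha dissolves the ripple.
# A static switch optionally multiplies this saturated mask by the borrowed SlurpVatMask function for world sphere masking.
# A DebugTimeSine call sits unconnected, and no material output connections appear.
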Material /Game/Packages/DS_Fortnite_Terrain_NoLOD/Water/Particles/M_SplashRipple
graph {
 properties {
  bUsedWithNanite=true
  bUsedWithNiagaraSprites=true
  bUsedWithParticleSprites=true
  blendMode=BLEND_Translucent
  pixelDepthOffsetMode=PDOM_Legacy
  translucencyLightingMode=TLM_VolumetricPerVertexNonDirectional
  usedShadingModels=Default Lit
 }
 nodes [
  MaterialFunctionCall() {materialFunction=&DebugTimeSine}
  n1 = ParticleColor()
  n2 = TextureSampleParameter2D() {parameterName=RippleTexture, texture=&/T_SplashRipple}
  n3 = Divide(A=Subtract(A=@n2.B, B=@n1.A), B=OneMinus(None=@n1.A))
  n4 = Saturate(None=Multiply(A=OneMinus(None=@n1.A), B=Multiply(A=@n2.R, B=@n3)))
  StaticSwitchParameter(True=Multiply(A=MaterialFunctionCall() {materialFunction=/Game/Athena/Items/Gameplay/SilkyVeteran/FX/SlurpVatMask}.Result, B=@n4), False=@n4) {parameterName=Use World Sphere Mask}
 ]
}
