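# Unlit, translucent, two-sided particle material for a cheap water splash ripple, usable on Niagara and Cascade sprites.
# Particle color times an intensity parameter drives emissive, optionally tinted through a time-of-day color function, then passed through an FX emissive adjustment function.
# Particle alpha multiplies a ripple texture mask for opacity.
# A static switch selects either the plain blue channel or the red channel with panning-noise UV distortion and a circular sphere mask.
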Material /Game/Packages/DS_Fortnite_Terrain_NoLOD/Water/Particles/M_SplashRipple_Cheap
graph {
 properties {
  bUsedWithNanite=true
  bUsedWithNiagaraSprites=true
  bUsedWithParticleSprites=true
  blendMode=BLEND_Translucent
  pixelDepthOffsetMode=PDOM_Legacy
  shadingModel=MSM_Unlit
  translucencyLightingMode=TLM_VolumetricPerVertexNonDirectional
  twoSided=true
  usedShadingModels=Unlit
 }
 nodes [
  n1 = TextureSample(UVs=Panner(Coordinate=TextureCoordinate() {uTiling=1.5, vTiling=1.5}) {speedY=0.3}) {samplerType=SAMPLERTYPE_Grayscale, texture=&LowResBlurredNoise}
  ConstantBiasScale(None=@n1.R) {bias=-0.5, scale=2}
  n2 = ParticleColor()
  n3 = Multiply(A=@n2.RGB, B=ScalarParameter() {defaultValue=1, parameterName=Emissive_Intensity})
  n4 = Multiply(A=Constant3Vector() {constant=<rgba 0.5647 1 0.9216 0>}, B=Multiply(A=@n3) {constB=0.7})
  n5 = Multiply(A=Constant3Vector() {constant=<rgba 0.5647 1 0.9216 0>}, B=@n3)
  n6 = Multiply(A=Constant3Vector() {constant=<rgba 0.5208 0.7293 1 0>})
  MaterialFunctionCall(Emissive=StaticSwitchParameter(
    True=MaterialFunctionCall(Color_Morning=@n4, Color_Day=@n5, Color_Evening=@n4, Color_Night=@n6) {materialFunction=&Color_Time_of_Day}.Result
    False=@n3
   ) {parameterName=Use_Color_Time_of_Day}) {materialFunction=&MF_FX_EmissiveAdjustments}
  n7 = TextureCoordinate()
  n8 = Multiply(A=@n1.R, B=ScalarParameter() {defaultValue=0.05, parameterName=DisturbanceIntensity}) {constB=0.05}
  n9 = TextureSample(UVs=Panner(Coordinate=Add(A=@n7, B=@n8), Time=Constant())) {samplerSource=SSM_Clamp_WorldGroupSettings, texture=&/T_SplashRipple}
  n10 = SphereMask(A=TextureCoordinate(), B=Constant2Vector() {g=0.5, r=0.5}, Radius=Constant() {r=0.5}, Hardness=Constant())
  n11 = TextureSample() {texture=&T_SplashRipple}
  StaticSwitchParameter(True=Multiply(A=@n2.A, B=Multiply(A=@n9.R, B=@n10)), False=Saturate(None=Multiply(A=@n2.A, B=@n11.B))) {parameterName=Use_Disturbance}
 ]
}
