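# M_WaterDrop is a default lit, opaque material for water drop particles, flagged for mesh particles, Niagara mesh particles and Nanite.
# It adds the particle color RGB to the inverted output of MF_Fresnel_FX with exponent 3, varying brightness by view angle.
# Loose constants of 0, 0.5 and 1 supply simple scalar values, and tangent-space normals are turned off.
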Material /Game/Packages/DS_Fortnite_Terrain_NoLOD/Water/Particles/M_WaterDrop
graph {
 properties {
  bTangentSpaceNormal=false
  bUsedWithMeshParticles=true
  bUsedWithNanite=true
  bUsedWithNiagaraMeshParticles=true
  numCustomizedUVs=4
  pixelDepthOffsetMode=PDOM_Legacy
  usedShadingModels=Default Lit
 }
 nodes [
  Add(
   A=ParticleColor().RGB
   B=MaterialFunctionCall(Exponent=Constant() {r=3}) {materialFunction=&MF_Fresnel_FX}.Fresnel Inverted
  )
  Constant()
  Constant() {r=0.5}
  Constant() {r=1}
 ]
}
